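# Sound cue for the looping tick of a grandfather clock loot container.
# A looping wave player feeds a modulator that adds slight pitch and volume variation, at a reduced cue volume in the loot collect sound class.
# Overridden attenuation uses logarithmic falloff with distance low-pass filtering, occlusion and listener focus.
# A shared concurrency rule, named for one instance per owner stopping the oldest, limits playback.
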SoundCue /Game/Sounds/Foley_Loot/Containers/Grandfather_Clock/Grandfather_Clock_Tick_Cue
root=Modulator_23
nodes [
 {id=Modulator_23, children [WavePlayer_0], properties {pitchMax=1, pitchMin=0.92, volumeMax=0.75, volumeMin=0.73}}
 {
  id=WavePlayer_0
  properties {bLooping=true, soundWaveAssetPtr=&/Grandfather_Clock_TickingLoop}
 }
]
properties {
 attenuationSettings=/Game/Sounds/Attenuation/Fort_Prop_TV_Att
 bOverrideAttenuation=true
 maxDistance=1400
 soundClassObject=&Fort_Loot_Collect
 volumeMultiplier=0.5
 attenuationOverrides {
  bAttenuateWithLPF=true
  bEnableListenerFocus=true
  bEnableOcclusion=true
  distanceAlgorithm=Logarithmic
  falloffDistance=1000
  focusAzimuth=25
  lPFRadiusMax=1500
  lPFRadiusMin=500
  nonFocusAzimuth=35
  nonFocusVolumeAttenuation=0.3
  nonSpatializedRadiusStart=100
  occlusionLowPassFilterFrequency=200
  occlusionTraceChannel=ECC_GameTraceChannel11
 }
 concurrencyOverrides {resolutionRule=PreventNew}
 concurrencySet [/Game/Sounds/Concurrency_Rules/OneMaximum_Owner_StopOldest]
}
